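# Translucent unlit material for the blue flame cone effect on the Airborne hoverplatform.
# A Fresnel, vertical UV gradient and depth fade mask drives opacity with particle alpha, and its cube blends dark to bright color.
# Static switches add an opacity multiplier and hue shift.
# A time-based sine makes a vertical offset scaled by WPOMult, alongside separate branches for tilted-axis rotation and corrected normals.
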Material /Game/Athena/Artemis/Environments/BuildingActors/Airborne/FX/Materials/M_Airborne_Hoverplatform_Flame_Cone
graph {
 properties {blendMode=BLEND_Translucent, pixelDepthOffsetMode=PDOM_Legacy, shadingModel=MSM_Unlit, usedShadingModels=Unlit}
 nodes [
  n1 = Time()
  n2 = Multiply(A=Multiply(A=Sine(None=@n1) {period=3.5}, B=Constant3Vector() {constant=<rgba 0 0 1 0>})) {constB=10}
  n3 = Normalize(VectorInput=Constant3Vector() {constant=<rgba -0.1 0.1 1 0>})
  n4 = Multiply(A=Time()) {constB=0.15}
  Add(
   A=@n2
   B=RotateAboutAxis(
    NormalizedRotationAxis=@n3
    RotationAngle=@n4
    PivotPoint=MaterialFunctionCall() {materialFunction=&ComponentPivotLocation}.Result
    Position=WorldPosition().XYZ
   )
  )
  Constant() {r=1}
  Reroute(None=Multiply(A=@n2, B=ScalarParameter() {defaultValue=1, parameterName=WPOMult}))
  n5 = Constant() {r=3}
  n6 = Power(Base=ComponentMask(None=TextureCoordinate()) {g=true}) {constExponent=4}
  n7 = Multiply(
   A=Multiply(
    A=MaterialFunctionCall(Exponent=@n5) {materialFunction=&MF_Fresnel_FX}.Fresnel
    B=Multiply(A=@n6) {constB=5}
   )
   B=DepthFade(FadeDistance=Constant() {r=12})
  )
  n8 = Multiply(A=@n7, B=ParticleColor().A)
  StaticSwitchParameter(True=Multiply(A=@n8, B=ScalarParameter() {defaultValue=1, parameterName=OpacityMult}), False=@n8) {parameterName=UseOpacity}
  n9 = LinearInterpolate(
   A=VectorParameter() {parameterName=DarkCol, defaultValue=<rgba 0 0.3333 2 0>}.RGB
   B=VectorParameter() {parameterName=BrightCol, defaultValue=<rgba 0.02 3.683 20 0>}.RGB
   Alpha=Power(Base=@n7) {constExponent=3}
  )
  StaticSwitchParameter(
   True=MaterialFunctionCall(Hue Shift Percentage=ScalarParameter() {parameterName=HueShiftPercent}, Texture=@n9) {materialFunction=&HueShift}.Result
   False=@n9
  ) {parameterName=UseHueShift}
  Transform(None=MaterialFunctionCall(Rotation Axis=@n3, Rotation Angle=@n4) {materialFunction=&FixRotateAboutAxisNormals}.New Tangent Space Vertex Normal)
 ]
}
